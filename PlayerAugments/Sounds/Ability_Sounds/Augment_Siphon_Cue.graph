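# Sound cue for the Siphon player augment, playing what its wave name indicates is a Siphon shotgun heal sound.
# A fixed short delay feeds a single wave player, so the sound starts slightly after it triggers.
# It uses the consumable sound class and consumable-use attenuation, and its concurrency override caps simultaneous instances, stopping the oldest.
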SoundCue /PlayerAugments/Sounds/Ability_Sounds/Augment_Siphon_Cue
root=Delay_0
nodes [
 {id=Delay_0, children [WavePlayer_0], properties {delayMax=0.2, delayMin=0.2}}
 {id=WavePlayer_0, properties {soundWaveAssetPtr=&/Augment_SiphonShotgun_Heal_01}}
]
properties {
 attenuationSettings=&Fort_Consumable_Use_Athena_Att
 bOverrideConcurrency=true
 maxDistance=2250
 soundClassObject=&Fort_Consumables
 volumeMultiplier=1
 concurrencyOverrides {maxCount=2, resolutionRule=StopOldest}
}
