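# Sound cue for the healing sound of the Siphon Shotgun player augment.
# It plays a single wave directly from the root, with no randomization or mixing.
# It uses consumable-use attenuation and the consumables sound class.
# A concurrency override limits it to one instance per owner, stopping the oldest when retriggered.
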SoundCue /PlayerAugments/Sounds/Ability_Sounds/SC_Augment_SiphonShotgun_Heal
root=WavePlayer_0
nodes [{id=WavePlayer_0, properties {soundWaveAssetPtr=&/Augment_SiphonShotgun_Heal_01}}]
properties {
 attenuationSettings=&Fort_Consumable_Use_Athena_Att
 bOverrideConcurrency=true
 maxDistance=2250
 soundClassObject=&Fort_Consumables
 concurrencyOverrides {bLimitToOwner=true, maxCount=1, resolutionRule=StopOldest}
}
